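# Lit master-style material for the New Year 2020 glider, adding an animated emissive 2020 display and fireworks over standard color, mask and normal maps.
# Pre-skinned position planar-projects the jittering 2020 and two firework layers, masked to upward-facing areas.
# Curve-atlas rows on a looping clock drive glow, colors and polar-coordinate streak-to-blast bursts; trigger parameters can override firework timing.
# The animated letters branch never reaches emissive.
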Material /BRCosmetics/Gadgets/Assets/VinderTech_GliderChute/Glider_New_Year_2020/Materials/M_Glider_NewYears2020
graph {
 properties {
  bUseMaterialAttributes=true
  bUsedWithMeshDeformer=true
  bUsedWithSkeletalMesh=true
  pixelDepthOffsetMode=PDOM_Legacy
  usedShadingModels=Default Lit
 }
 nodes [
  n1 = VertexInterpolator(VS=PreSkinnedPosition())
  n2 = VectorParameter() {parameterName=LargeFireworposition, defaultValue=<rgba 0 0 0 1>}
  n3 = ComponentMask(None=Divide(A=Add(A=@n1.PS, B=@n2.RGB), B=VectorParameter() {parameterName=LargeFireworkScale, defaultValue=<rgba 512 512 512 1>}.RGB)) {g=true, r=true}
  n4 = TextureObject() {samplerType=SAMPLERTYPE_Masks, texture=/BRCosmetics/Gadgets/Assets/VinderTech_GliderChute/Glider_New_Year_2020/Textures/T_LargeFireWork}
  n5 = TextureSample!(
   UVs=@n3
   Tex
   Level=MaterialFunctionCall!(UVs=@n3, Texture Size=TextureProperty(Tex=@n4)) {materialFunction=/Engine/Functions/Engine_MaterialFunctions01/Texturing/ComputeMipLevel}.MipLevel
   Apply View MipBias
  ) {
   mipValueMode=TMVM_MipLevel
   samplerType=SAMPLERTYPE_Masks
   texture=&T_LargeFireWork
  }
  n6 = Time()
  Add(A=@n5.B, B=@n6)
  n7 = VectorParameter() {parameterName=Position, defaultValue=<rgba 48 48 0 1>}
  n8 = VectorParameter() {parameterName=2020Scale, defaultValue=<rgba 256 256 256 1>}
  n9 = ComponentMask(None=Divide(A=Add(A=@n1.PS, B=@n7.RGB), B=@n8.RGB)) {g=true, r=true}
  n10 = TextureSample!(UVs=@n9, Tex, Apply View MipBias) {samplerType=SAMPLERTYPE_Masks, texture=/BRCosmetics/Gadgets/Assets/VinderTech_GliderChute/Glider_New_Year_2020/Textures/T_2020_NewYear}
  n11 = Constant2Vector() {g=0.5, r=0.5}
  n12 = Round(None=Abs(None=Subtract(A=@n9, B=@n11)))
  n13 = Constant() {r=1}
  n14 = OneMinus(None=Ceil(None=Desaturation(None=@n12, Fraction=@n13)))
  n15 = VectorParameter() {parameterName=2020Color, defaultValue=<rgba 1 0.5011 0 1>}
  n16 = Multiply(A=ComponentMask(None=Multiply(A=@n10.RGB, B=@n14)) {g=true}, B=@n15.RGB)
  n17 = ScalarParameter() {defaultValue=5, parameterName=BG_GlowStrength}
  n18 = Time()
  n19 = Constant() {r=0.5}
  n20 = AppendVector(
   A=Multiply(A=@n18, B=ScalarParameter() {defaultValue=0.1, parameterName=Speed})
   B=Multiply(A=Add(A=@n18, B=@n19), B=ScalarParameter() {defaultValue=0.5, parameterName=Speed2})
  )
  n21 = Add(A=@n18) {constB=0.25}
  n22 = MaterialFunctionCall(Vec3=Frac(None=AppendVector(A=@n20, B=@n21))) {materialFunction=&MF_Hash}
  n23 = ScalarParameter() {defaultValue=0.01, parameterName=Mix_Offset_amt}
  n24 = ScalarParameter() {defaultValue=0.05, parameterName=Max_Offset_amt}
  n25 = Multiply(A=Time(), B=ScalarParameter() {defaultValue=0.1, parameterName=GlobalAnimationSpeed})
  n26 = CurveAtlasRowParameter(CurveTime=Frac(None=@n25)) {
   atlas=&AthenaCurveAtlas01
   curve=/Game/Gadgets/Assets/VinderTech_GliderChute/Glider_New_Year_2020/Textures/NYjittermovementcurve
   defaultValue=178
   parameterName=JitterAnimCurve
  }
  n27 = ComponentMask(None=@n26) {r=true}
  n28 = Add(A=@n9, B=Multiply(A=ComponentMask(None=@n22.Vec3) {g=true, r=true}, B=LinearInterpolate(A=@n23, B=@n24, Alpha=@n27)))
  n29 = TextureSample(UVs=@n28) {samplerType=SAMPLERTYPE_Masks, texture=&T_2020_NewYear}
  n30 = Constant2Vector() {g=0.5, r=0.5}
  n31 = Round(None=Abs(None=Subtract(A=@n28, B=@n30)))
  n32 = Constant() {r=1}
  n33 = OneMinus(None=Ceil(None=Desaturation(None=@n31, Fraction=@n32)))
  n34 = Multiply(A=ComponentMask(None=Multiply(A=@n29.RGB, B=@n33)) {r=true}, B=@n15.RGB)
  n35 = ScalarParameter() {defaultValue=500, parameterName=2020_BloomGlowStrength}
  n36 = ComponentMask(None=@n26) {b=true}
  n37 = LinearInterpolate(A=ScalarParameter() {defaultValue=10, parameterName=2020_GlowStrength}, B=@n35, Alpha=@n36) {constAlpha=0}
  n38 = ScalarParameter() {parameterName=UseTriggerTiming}
  n39 = LinearInterpolate(A=@n36, B=@n38, Alpha=@n38)
  n40 = ScalarParameter() {defaultValue=2, parameterName=BIgFireWorkSpeed}
  n41 = ScalarParameter() {parameterName=TriggerTiming}
  n42 = LinearInterpolate(A=Frac(None=Multiply(A=@n6, B=@n40)), B=@n41, Alpha=@n38)
  n43 = OneMinus(None=Divide(A=Distance(A=@n5.R, B=@n42), B=ScalarParameter() {defaultValue=0.15, parameterName=FireWorkSize}))
  n44 = CurveAtlasRowParameter(CurveTime=@n5.B) {
   atlas=&AthenaCurveAtlas01
   curve=/Game/Gadgets/Assets/VinderTech_GliderChute/Glider_New_Year_2020/Textures/NYfireworkscolorcurve
   defaultValue=176
   parameterName=FireWorksColorCurve
  }
  n45 = Multiply(A=Multiply(A=@n5.G, B=Clamp(None=@n43)), B=@n44)
  n46 = LinearInterpolate(A=ScalarParameter() {defaultValue=20, parameterName=FireWorksGlowStrength}, B=@n35, Alpha=@n36)
  n47 = Add(
   A=Add(A=Multiply(A=@n16, B=@n17), B=Multiply(A=@n34, B=@n37))
   B=Multiply(A=Multiply(A=@n39, B=@n45) {constA=1}, B=Multiply(A=@n5.B, B=@n46))
  )
  n48 = Add(A=@n7.RGB, B=VectorParameter() {parameterName=TextPosition, defaultValue=<rgba 0 0 0 1>}.RGB)
  n49 = VectorParameter() {parameterName=TextScale, defaultValue=<rgba 1 1 1 1>}
  n50 = ComponentMask(None=Divide(A=Add(A=@n1.PS, B=@n48), B=Multiply(A=@n8.RGB, B=@n49.RGB))) {g=true, r=true}
  n51 = MaterialFunctionCall(Float2=@n50) {materialFunction=&BreakOutFloat2Components}
  n52 = CurveAtlasRowParameter(CurveTime=Frac(None=@n25)) {
   atlas=&AthenaCurveAtlas01
   curve=/Game/Gadgets/Assets/VinderTech_GliderChute/Glider_New_Year_2020/Textures/NYLettersAnimCurve
   defaultValue=180
   parameterName=LettersAnimCurve
  }
  n53 = ComponentMask(None=@n52) {g=true}
  n54 = TextureSample(UVs=@n9) {samplerType=SAMPLERTYPE_Masks, texture=&T_2020_NewYear}
  n55 = ComponentMask(None=Clamp(None=@n54.RGBA) {minDefault=0.01}) {a=true}
  n56 = ScalarParameter() {defaultValue=0.1, parameterName=LettersOffsetBias}
  n57 = Multiply(A=Multiply(A=@n53, B=ScalarParameter() {defaultValue=0.5, parameterName=LettersOffsetDistance}), B=Clamp(None=Multiply(A=@n55, B=@n56)))
  n58 = MaterialFunctionCall(X=@n51.R, Y=LinearInterpolate(A=@n51.G, B=Add(A=@n51.G, B=@n57), Alpha=@n53)) {materialFunction=&MakeFloat2}
  n59 = TextureObject() {texture=&T_2020_NewYear}
  n60 = TextureSample!(
   UVs=@n58.Result
   Tex
   Level=MaterialFunctionCall(UVs=@n50, Texture Size=TextureProperty(Tex=@n59)) {materialFunction=&ComputeMipLevel}.MipLevel
   Apply View MipBias
  ) {
   mipValueMode=TMVM_MipLevel
   samplerType=SAMPLERTYPE_Masks
   texture=&T_2020_NewYear
  }
  n61 = Constant2Vector() {g=0.5, r=0.5}
  n62 = Round(None=Abs(None=Subtract(A=@n50, B=@n61)))
  n63 = Constant() {r=1}
  n64 = OneMinus(None=Ceil(None=Desaturation(None=@n62, Fraction=@n63)))
  n65 = TextureObject() {samplerType=SAMPLERTYPE_Masks, texture=/BRCosmetics/Gadgets/Assets/VinderTech_GliderChute/Glider_New_Year_2020/Textures/T_NY_FireWorks_01}
  n66 = TextureSample(
   UVs=@n58.Result
   Level=MaterialFunctionCall(UVs=@n50, Texture Size=TextureProperty(Tex=@n65)) {materialFunction=&ComputeMipLevel}.MipLevel
  ) {
   mipValueMode=TMVM_MipLevel
   samplerType=SAMPLERTYPE_Masks
   texture=&T_NY_FireWorks_01
  }
  n67 = Multiply(
   A=LinearInterpolate(A=Multiply(A=@n60.B, B=@n64), B=Multiply(A=@n66.A, B=@n64), Alpha=ComponentMask(None=@n52) {b=true})
   B=ComponentMask(None=@n52) {r=true}
  )
  n68 = ScalarParameter() {defaultValue=10, parameterName=LetterGlowStrength}
  Add(A=@n47, B=Multiply(A=Multiply(A=@n67, B=@n15.RGB), B=LinearInterpolate(A=@n68, B=@n35, Alpha=@n36)))
  Frac(None=Multiply(A=Time(), B=ScalarParameter() {defaultValue=0.1, parameterName=BlastSpeed}))
  n69 = TextureSampleParameter2D() {parameterName=BaseColor, texture=/BRCosmetics/Gadgets/Assets/VinderTech_GliderChute/Glider_New_Year_2020/Textures/T_Glider_NewYear_2020_D}
  n70 = TextureSampleParameter2D() {
   parameterName=_S Map
   samplerType=SAMPLERTYPE_Masks
   texture=/BRCosmetics/Gadgets/Assets/VinderTech_GliderChute/Glider_New_Year_2020/Textures/T_Glider_NewYear_2020_S
  }
  n71 = TextureSample!(UVs=@n9, Tex, Apply View MipBias) {samplerType=SAMPLERTYPE_Masks, texture=&T_NY_FireWorks_01}
  n72 = Multiply(A=Add(A=@n71.B, B=Time()), B=ScalarParameter() {defaultValue=1, parameterName=FireWorkSpeed})
  n73 = Distance(A=@n71.R, B=LinearInterpolate(A=Frac(None=@n72), B=@n41, Alpha=@n38))
  n74 = ScalarParameter() {defaultValue=0.15, parameterName=FireWorkSize}
  n75 = Clamp(None=OneMinus(None=Divide(A=@n73, B=@n74)))
  n76 = CurveAtlasRowParameter(CurveTime=@n71.B) {
   atlas=&AthenaCurveAtlas01
   curve=&NYfireworkscolorcurve
   defaultValue=176
   parameterName=FireWorksColorCurve
  }
  n77 = Multiply(A=LinearInterpolate(A=ComponentMask(None=@n26) {g=true}, B=@n38, Alpha=@n38), B=Multiply(A=Multiply(A=@n71.G, B=@n75), B=@n76))
  n78 = Multiply(A=@n71.B, B=ScalarParameter() {defaultValue=20, parameterName=FireWorksGlowStrength})
  n79 = TextureSample() {samplerType=SAMPLERTYPE_Masks, texture=/BRCosmetics/Gadgets/Assets/VinderTech_GliderChute/Glider_New_Year_2020/Textures/T_NewYears_2020_FX}
  n80 = Ceil(None=ComponentMask(None=VertexNormalWS()) {b=true})
  n81 = Multiply(A=Add(A=@n47, B=Multiply(A=@n77, B=@n78)), B=Multiply(A=@n79.R, B=Clamp(None=@n80)))
  n82 = MaterialFunctionCall() {materialFunction=&Pi}
  n83 = Subtract(A=Multiply(A=@n9) {constB=2})
  n84 = ComponentMask(None=@n83) {g=true}
  n85 = VectorParameter() {parameterName=Origin, defaultValue=<rgba 0 0 0 1>}
  n86 = ComponentMask(None=@n85.RGB) {g=true}
  n87 = ComponentMask(None=@n83) {r=true}
  n88 = ComponentMask(None=@n85.RGB) {r=true}
  n89 = MaterialFunctionCall() {materialFunction=&Pi}
  n90 = Add(A=Arctangent2Fast(Y=Subtract(A=@n84, B=@n86), X=Subtract(A=@n87, B=@n88)), B=Add(A=Divide(A=@n89.Result), B=@n89.Result))
  n91 = MaterialFunctionCall(Multiplier=Constant() {r=2}) {materialFunction=&Pi}
  n92 = Subtract(A=@n83, B=ComponentMask(None=@n85.RGB) {g=true, r=true})
  n93 = AppendVector(
   A=Multiply(A=Divide(B=@n82.Result) {constA=2}, B=Fmod(A=@n90, B=@n91.Result))
   B=Multiply(A=MaterialFunctionCall(Vector 2=@n92) {materialFunction=/Engine/Functions/Engine_MaterialFunctions02/Utility/VectorLength}.V2 Length) {constB=-3}
  )
  n94 = Constant2Vector() {g=0.1, r=0.5}
  n95 = TextureSample(UVs=Panner(Coordinate=Multiply(A=@n93, B=@n94)) {speedY=-2}) {samplerType=SAMPLERTYPE_Masks, texture=/BRCosmetics/Gadgets/Assets/VinderTech_GliderChute/Glider_New_Year_2020/Textures/T_Blast}
  n96 = Multiply(
   A=Multiply(A=ComponentMask(None=@n95.RGB) {r=true}, B=VectorParameter() {parameterName=BuildupStreaksColor, defaultValue=<rgba 1 0 0.9475 1>}.RGB)
   B=ScalarParameter() {defaultValue=10, parameterName=BuildupStreaksGlowStrength}
  )
  n97 = Constant2Vector() {g=0.05, r=0.15}
  n98 = VectorParameter() {parameterName=ColorBlastSpeed, defaultValue=<rgba 0 -1 0 1>}
  n99 = TextureSample(UVs=Panner(Coordinate=Multiply(A=@n93, B=@n97), Speed=ComponentMask(None=@n98.RGB) {g=true, r=true}) {speedY=0.5}) {samplerType=SAMPLERTYPE_Masks, texture=&T_Blast}
  n100 = Add(A=ComponentMask(None=@n99.RGB) {g=true}, B=@n99.R)
  n101 = Constant2Vector() {g=0.2, r=0.5}
  n102 = VectorParameter() {parameterName=ColorBlastMaskSpeed, defaultValue=<rgba 0 -1 0 1>}
  n103 = TextureSample(UVs=Panner(Coordinate=Multiply(A=@n93, B=@n101), Speed=ComponentMask(None=@n102.RGB) {g=true, r=true}) {speedY=-1}) {texture=/Game/Characters/Player/Male/Medium/Bodies/M_MED_Viper/Textures/Glitch/T_ViperGlitchNoise}
  n104 = CurveAtlasRowParameter(CurveTime=Clamp(None=Subtract(A=@n100, B=@n103.R))) {
   atlas=&AthenaCurveAtlas01
   curve=/Game/Gadgets/Assets/VinderTech_GliderChute/Glider_New_Year_2020/Textures/NYBlastColor
   defaultValue=174
   parameterName=Blast_ColorCurve
  }
  n105 = Panner(Coordinate=Multiply(A=@n93, B=Constant2Vector() {g=0.05, r=0.15})) {speedY=1.5}
  n106 = Multiply(
   A=Multiply(
    A=Multiply(A=@n104, B=@n100)
    B=TextureSample(UVs=@n105) {samplerType=SAMPLERTYPE_Masks, texture=&T_Blast}.B
   )
   B=ScalarParameter() {defaultValue=20, parameterName=BlastBrightness}
  )
  n107 = CurveAtlasRowParameter(CurveTime=Frac(None=@n25)) {
   atlas=&AthenaCurveAtlas01
   curve=/Game/Gadgets/Assets/VinderTech_GliderChute/Glider_New_Year_2020/Textures/NYblasttimingcurve
   defaultValue=173
   parameterName=BlastTimingCurve
  }
  n108 = ComponentMask(None=@n107) {r=true}
  n109 = Multiply(A=@n79.G, B=Multiply(A=LinearInterpolate(A=@n96, B=@n106, Alpha=@n108), B=ComponentMask(None=@n107) {b=true}))
  n110 = TextureSampleParameter2D() {
   parameterName=NormalMap
   samplerType=SAMPLERTYPE_Normal
   texture=/BRCosmetics/Gadgets/Assets/VinderTech_GliderChute/Glider_New_Year_2020/Textures/T_Glider_NewYear_2020_N
  }
  n111 = SetMaterialAttributes!(
   MaterialAttributes
   Base Color=NamedRerouteDeclaration(None=@n69.RGB) {name=baseColro, nodeColor=<rgba 0 1 0.7188 1>}
   Metallic=NamedRerouteDeclaration(None=@n70.G) {name=metallic, nodeColor=<rgba 0.0312 0 1 1>}
   Specular=NamedRerouteDeclaration(None=@n70.R) {name=specular, nodeColor=<rgba 0 0.8594 1 1>}
   Roughness=NamedRerouteDeclaration(None=@n70.B) {name=roughness, nodeColor=<rgba 0 1 0.8125 1>}
   Emissive Color=NamedRerouteDeclaration(None=Add(A=@n81, B=@n109)) {name=emissive, nodeColor=<rgba 0 0.6016 1 1>}
   Normal=NamedRerouteDeclaration(None=@n110.RGB) {name=normal, nodeColor=<rgba 0.3359 0 1 1>}
  ) {attributeSetTypes [
    #1
    #2
    #3
    #4
    #5
    #6
   ]}
  MaterialFunctionCall(In=@n111) {materialFunction=&MF_TechArtOverrides_accessories}
  Multiply(A=@n24, B=Clamp(None=@n27) {minDefault=0.02})
  Multiply(A=TextureSample(UVs=@n9) {texture=/Game/Weapons/WeaponSkins/Wraps/Textures/EvilAstronaut/T_EvilAstronaut_Logo_Wrap_02}.RGB, B=@n33)
  SceneTexelSize()
  TextureCoordinate()
  TextureCoordinate()
  TextureCoordinate()
  VectorParameter() {parameterName=LettersGlowCoilor, defaultValue=<rgba 0 0 0 1>}
 ]
}
